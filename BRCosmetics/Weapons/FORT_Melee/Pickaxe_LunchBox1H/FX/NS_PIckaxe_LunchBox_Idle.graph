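# Idle Niagara effect that puts crackling yellow electricity on the one-handed LunchBox pickaxe.
# Handle and head emitters burst short-lived stroke sprites with soft glow sprites, re-positioned and oriented each update from a sampled skeletal mesh bone or socket, alongside a looping flash light.
# Partitioned ribbon arcs follow sockets through BoneFlow, flickering via curves on looping bursts.
# Front-end modifiers halve brightness and shift a head glow offset for menu display.
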NiagaraSystem /BRCosmetics/Weapons/FORT_Melee/Pickaxe_LunchBox1H/FX/NS_PIckaxe_LunchBox_Idle
dependencies {
 usedDataInterfaces [&NiagaraDataInterfaceColorCurve, &NiagaraDataInterfaceCurve]
 usedDynamicInputs []
 usedModules [&SystemState]
}
emitters {
 Electric_Strokes_Handle {
  emitterUpdateScript [
   &EmitterState()
   &SpawnBurst_Instantaneous(Spawn Count=&UniformRangedInt(Minimum=2, Maximum=3), Use Spawn Probability=false)
  ]
  particleSpawnScript [
   /Niagara/Modules/Spawn/Initialization/InitializeParticle(
    Color=<rgba 1 0.8185 0 1>
    Lifetime=&UniformRangedFloat(Minimum=0.25, Maximum=0.35)
    Module.Write Color=true
    Module.Write Lifetime=true
    Module.Write Mass=true
    Module.Write Material Random=true
    Module.Write Position=false
    Module.Write Scale=false
    Module.Write SpriteRotation=true
    Module.Write SpriteSize=true
    Sprite Size=<25 66>
   )
   &DynamicMaterialParameters(
    Index 0 Param 1=&UniformRangedFloat(Minimum=2.5, Maximum=3.5)
    Index 0 Param 3=&UniformRangedFloat(Minimum=0.2, Maximum=0.5)
    Index 2 Param 1=0.2
    Index 2 Param 2=1
    Module.Param0WriteEnabled=true
    Module.Param0WriteEnabled2=true
    Module.Param1WriteEnabled=false
    Module.Param1WriteEnabled2=true
    Module.Param2WriteEnabled=true
    Module.Param2WriteEnabled2=true
    Module.Param3WriteEnabled=true
    Module.Param3WriteEnabled2=true
    Write Parameter Index 2=true
   )
   &CameraOffset(Camera Offset Amount=&RandomRangeFloat(Minimum=-2, Maximum=2))
   /SetVariables(Particles.UVScale:Vector2f=&MakeVector2D(X=&UniformAOrBFloat(A=-1), Y=&UniformAOrBFloat(A=-1))) {isSetParameters=true, module=SetVariables_#1}
  ]
  particleUpdateScript [
   &SkeletalMeshLocation(
    Apply Sampled Orientation to Mesh Particles=true
    Bone / Socket Index=9
    Inherit Velocity=false
    MaskSpawnGroup=false
    Mesh Sampling Type=::NewEnumerator5
    Skeletal Mesh=$User.SkeletalMesh
    Socket Sampling Mode=::NewEnumerator5
   )
   /SetVariables(Particles.SpriteAlignment=&RotateVectorByQuaternion(Vector To Rotate=<-0.035 0 1>, Quaternion=$Output.SkeletalMeshLocation.SampledRotation)) {isSetParameters=true, module=SetVariables_#2}
   /SetVariables(Particles.Position=&Add_Vector(
     A=$Output.SkeletalMeshLocation.SampledPosition
     B=&RotateVectorByQuaternion(Vector To Rotate=<6.75 0 2>, Quaternion=$Output.SkeletalMeshLocation.SampledRotation)
    )) {isSetParameters=true}
   &ParticleState()
   &ScaleSpriteSize(Scale Factor=&Vector2DFromFloat(Value=&FloatFromCurve(FloatCurve=DataInterface() {propertyValues {Curve {keys [{arriveTangent=7.3705, interpMode=RCIM_Cubic, leaveTangent=7.3705, tangentMode=RCTM_User, time=0, value=0.3}, <at 0.3 1 RCIM_Cubic>]}}})))
   &Color(
    Color=&ColorFromCurve(ColorCurve=DataInterface() {propertyValues {
       AlphaCurve {keys [<at 0 0 RCIM_Cubic>,<at 0.1239 1>,<at 0.346 1>,<at 1 0 RCIM_Cubic>]}
       BlueCurve {keys [<at -0.006 4>,<at 0.261 1.2>]}
       GreenCurve {keys [<at -0.006 15.353>,<at 0.261 4.6059>]}
       RedCurve {keys [<at -0.006 20>,<at 0.261 6>]}
      }})
    Scale Color=&VectorFromFloat(Value=&FN_FrontEndFloatModifier(Gameplay=1, Front End=0.5))
   )
   /Game/Effects/Niagara/Modules/Sprite/FN_Sprite_v2(
    Color:LinearColor=&MakeLinearColorFromVectorAndFloat(
     "Vector (RGB)"=&MakeVectorFromLinearColorRGB(Color=<rgba 0.75 0.6662 0.2344 0.175>)
     "Float (Alpha)"=&Lerp_Float(A=0.075, B=0, Alpha=$Particles.NormalizedAge)
    )
    Create Color Attribute=true
    Create Position Attribute=true
    Create SpriteSize Attribute=true
    Position:Vector3f=&Add_Vector(
     A=$Output.SkeletalMeshLocation.SampledPosition
     B=&RotateVectorByQuaternion(Vector To Rotate=<6.75 0 -12>, Quaternion=$Output.SkeletalMeshLocation.SampledRotation)
    )
    SpriteSize:Vector2f=<30 70>
   )
  ]
  rendererClasses [&NiagaraSpriteRendererProperties]
  renderers [
   {rendererClass=&NiagaraSpriteRendererProperties}
   {rendererClass=&NiagaraSpriteRendererProperties, rendererIndex=1}
  ]
 }
 Electric_Strokes_Head {
  emitterUpdateScript [
   &EmitterState()
   &SpawnBurst_Instantaneous(Spawn Count=&UniformRangedInt(Minimum=2, Maximum=3), Use Spawn Probability=false)
   /SetVariables(
    Emitter.FEGlow:NiagaraBool=$User.bIsFrontEndPreview
    Emitter.GlowRend:NiagaraBool=&InvertBool(Boolean:NiagaraBool=$User.bIsFrontEndPreview)
   ) {isSetParameters=true}
  ]
  particleSpawnScript [
   &InitializeParticle(
    Color=&LerpLinearColors(
     StartColor=<rgba 1 0.8185 0 1>
     EndColor=<rgba 0.7292 0.5968 0 1>
     LerpFactor=&FN_FrontEndFloatModifier(Gameplay=0, Front End=1)
    )
    Lifetime=&UniformRangedFloat(Minimum=0.25, Maximum=0.35)
    Module.Write Color=true
    Module.Write Lifetime=true
    Module.Write Mass=true
    Module.Write Material Random=true
    Module.Write Position=false
    Module.Write Scale=false
    Module.Write SpriteRotation=true
    Module.Write SpriteSize=true
    Sprite Size=<35 38>
   )
   &DynamicMaterialParameters(
    Index 0 Param 1=&UniformRangedFloat(Minimum=2.5, Maximum=3.5)
    Index 0 Param 3=&UniformRangedFloat(Minimum=0.2, Maximum=0.5)
    Index 2 Param 1=0.2
    Index 2 Param 2=1
    Module.Param0WriteEnabled=true
    Module.Param0WriteEnabled2=true
    Module.Param1WriteEnabled=false
    Module.Param1WriteEnabled2=true
    Module.Param2WriteEnabled=true
    Module.Param2WriteEnabled2=true
    Module.Param3WriteEnabled=true
    Module.Param3WriteEnabled2=true
    Write Parameter Index 2=true
   )
   &CameraOffset(Camera Offset Amount=&RandomRangeFloat(Minimum=-2, Maximum=2))
   /SetVariables(Particles.UVScale:Vector2f=&MakeVector2D(X=&UniformAOrBFloat(A=-1), Y=&UniformAOrBFloat(A=-1))) {isSetParameters=true, module=SetVariables_#1}
  ]
  particleUpdateScript [
   &SkeletalMeshLocation(
    Apply Sampled Orientation to Mesh Particles=true
    Bone / Socket Index=9
    Inherit Velocity=false
    MaskSpawnGroup=false
    Mesh Sampling Type=::NewEnumerator5
    Skeletal Mesh=$User.SkeletalMesh
    Socket Sampling Mode=::NewEnumerator5
   )
   /SetVariables(
    Particles.SpriteAlignment=&RotateVectorByQuaternion(Vector To Rotate=<-1 0 0>, Quaternion=$Output.SkeletalMeshLocation.SampledRotation)
    Particles.SpriteFacing=&RotateVectorByQuaternion(Vector To Rotate=<0 1 0>, Quaternion=$Output.SkeletalMeshLocation.SampledRotation)
   ) {isSetParameters=true, module=SetVariables_#2}
   &ParticleState()
   &ScaleSpriteSize(Scale Factor=&Vector2DFromFloat(Value=&FloatFromCurve(FloatCurve=DataInterface() {propertyValues {Curve {keys [{arriveTangent=7.3705, interpMode=RCIM_Cubic, leaveTangent=7.3705, tangentMode=RCTM_User, time=0, value=0.3}, <at 0.3 1 RCIM_Cubic>]}}})))
   &Color(
    Color=&ColorFromCurve(ColorCurve=DataInterface() {propertyValues {
       AlphaCurve {keys [<at 0 0 RCIM_Cubic>,<at 0.1239 1>,<at 0.346 1>,<at 1 0 RCIM_Cubic>]}
       BlueCurve {keys [<at -0.006 4>,<at 0.261 1.2>]}
       GreenCurve {keys [<at -0.006 15.353>,<at 0.261 4.6059>]}
       RedCurve {keys [<at -0.006 20>,<at 0.261 6>]}
      }})
    Scale Color=&VectorFromFloat(Value=&FN_FrontEndFloatModifier(Gameplay=1, Front End=0.5))
   )
   &FN_Sprite_v2(
    Color:LinearColor=&MakeLinearColorFromVectorAndFloat(
     "Vector (RGB)"=&MakeVectorFromLinearColorRGB(Color=<rgba 0.75 0.6662 0.2344 0.175>)
     "Float (Alpha)"=&Lerp_Float(A=0.075, B=0, Alpha=$Particles.NormalizedAge)
    )
    Create Color Attribute=true
    Create Position Attribute=true
    Create SpriteSize Attribute=true
    Position:Vector3f=&Add_Vector(
     A=$Output.SkeletalMeshLocation.SampledPosition
     B=&RotateVectorByQuaternion(Vector To Rotate=<-2 0 -2>, Quaternion=$Output.SkeletalMeshLocation.SampledRotation)
    )
    SpriteSize:Vector2f=<42.5 55>
   )
   &FN_Sprite_v2(
    Color:LinearColor=&MakeLinearColorFromVectorAndFloat(
     "Vector (RGB)"=&MakeVectorFromLinearColorRGB(Color=<rgba 0.75 0.6662 0.2344 0.175>)
     "Float (Alpha)"=&Lerp_Float(A=0.075, B=0, Alpha=$Particles.NormalizedAge)
    )
    Create Color Attribute=true
    Create Position Attribute=true
    Create SpriteSize Attribute=true
    SpriteSize:Vector2f=<42.5 55>
    Write to Module Namespace=true
   ) {enabled=false, module=FN_Sprite_v2001}
   &FN_Sprite_Attributes(
    Create Position Attribute=true
    Position:NiagaraPosition=&Add_Vector(
     A=$Output.SkeletalMeshLocation.SampledPosition
     B=&RotateVectorByQuaternion(
      Vector To Rotate=&FN_FrontEndVector3Modifier(Gameplay=<-2 0 -2>, Front End=<17 0 11.5>)
      Quaternion=$Output.SkeletalMeshLocation.SampledRotation
     )
    )
    Write to Module Namespace=true
   )
  ]
  rendererClasses [&NiagaraSpriteRendererProperties]
  renderers [
   {rendererClass=&NiagaraSpriteRendererProperties}
   {rendererClass=&NiagaraSpriteRendererProperties, rendererIndex=1}
   {rendererClass=&NiagaraSpriteRendererProperties, rendererIndex=2}
   {rendererClass=&NiagaraSpriteRendererProperties, rendererIndex=3}
  ]
 }
 Flash_Flare {
  emitterUpdateScript [
   &EmitterState(Life Cycle Mode=::NewEnumerator1, Loop Behavior=::NewEnumerator1, Loop Duration=0.2, UseLoopDelay=false)
   &SpawnBurst_Instantaneous(Spawn Count=1, Use Spawn Probability=false)
  ]
  particleSpawnScript [
   &InitializeParticle(
    Lifetime=&UniformRangedFloat(Minimum=0.25, Maximum=0.35)
    Module.Write Color=true
    Module.Write Lifetime=true
    Module.Write Mass=true
    Module.Write Material Random=true
    Module.Write Position=false
    Module.Write Scale=false
    Module.Write SpriteRotation=true
    Module.Write SpriteSize=true
   )
   /SetVariables(Particles.select:NiagaraFloat=&UniformRangedFloat(Minimum=0, Maximum=1)) {isSetParameters=true}
  ]
  particleUpdateScript [
   &SkeletalMeshLocation(
    Bone / Socket Index=9
    Inherit Velocity=false
    MaskSpawnGroup=false
    Mesh Sampling Type=::NewEnumerator5
    Skeletal Mesh=$User.SkeletalMesh
    Socket Sampling Mode=::NewEnumerator5
   )
   &ParticleState(Kill Particles When Lifetime Has Elapsed=false, Loop Particles Lifetime=true)
   &FN_Light_v2(
    Light Exponent=6
    Light Radius=45
    LightEnabled=true
    Position Offset:Vector3f=&RotateVectorByQuaternion(Vector To Rotate=<10 0 0>, Quaternion=$Output.SkeletalMeshLocation.SampledRotation)
   )
   /SetVariables(Particles.Light Color:LinearColor=&MultiplyLinearColor(
     Linear Color=&ColorFromCurve(ColorCurve=DataInterface() {propertyValues {
        AlphaCurve {keys [<at 0 0 RCIM_Cubic>,<at 0.1239 1>,<at 0.346 1>,<at 1 0 RCIM_Cubic>]}
        BlueCurve {keys [<at -0.006 3.6667>,<at 0.261 2.6641>,<at 0.795 0.4844>]}
        GreenCurve {keys [<at -0.006 6.7414>,<at 0.261 5.1177>,<at 0.795 0.9305>]}
        RedCurve {keys [<at -0.006 8>,<at 0.261 5.5>,<at 0.795 1>]}
       }})
     Scale Factor=&MakeVector4_FromFloat(Float=&FN_FrontEndFloatModifier(Gameplay=1, Front End=0.5))
    )) {isSetParameters=true}
  ]
  rendererClasses [&NiagaraLightRendererProperties]
  renderers [{rendererClass=&NiagaraLightRendererProperties, rendererIndex=0}]
 }
 LightningOuterArcs {
  emitterSpawnScript [
   /SetVariables(Emitter.RandomSeed:NiagaraInt32=/Niagara/DynamicInputs/Random/RandomInt(Randomness Mode:ENiagaraRandomnessMode=::NewEnumerator0, Random Seed:NiagaraInt32=0, Fixed Random Seed:NiagaraBool=false)) {isSetParameters=true}
   /Game/Effects/Niagara/Modules/Partition/FN_PartitionAttributes(Number of Bone Sets=3, Particles Per Partition=8)
  ]
  emitterUpdateScript [
   &EmitterState(
    Life Cycle Mode=::NewEnumerator1
    Loop Duration=&UniformRangedFloat(Minimum=0.5, Maximum=1.25)
    Recalculate Duration Each Loop=true
    UseLoopDelay=false
   )
   &SpawnBurst_Instantaneous(Spawn Count=$Emitter.FN_PartitionAttributes.Spawn Total, Use Spawn Probability=false)
  ]
  particleSpawnScript [
   /Game/Effects/Niagara/Modules/Partition/FN_Particle_partition_v5/FN_Particle_Partition_v5(Enable Ribbon ID=true, Seed:NiagaraInt32=$Emitter.RandomSeed, Shuffle Partition ID=true)
   &InitializeRibbon(
    Lifetime=&UniformRangedFloat(
     Minimum=0.2
     Maximum=0.3
     Randomness Mode=::NewEnumerator1
     Random Seed=&Add_Integer(A=$Particles.Partition.ID, B=$Emitter.LoopCount)
     Fixed Random Seed=true
    )
    Module.Write Color=true
    Module.Write Lifetime=true
    Module.Write Mass=true
    Module.Write Material Random=true
    Module.Write Position=true
    Module.Write RibbonTwist=false
    Module.Write RibbonWidth=true
    Ribbon Width=30
   )
   /Game/Effects/Niagara/Modules/BoneFlow/FN_Initialize_BoneFlow_Attributes_v2(
    Angle:NiagaraFloat=&Multiply_Float(
     A=$Particles.Partition.NormalizedID
     B=&UniformRangedFloat(Minimum=0.75, Maximum=1, Random Seed=$Particles.Partition.ID)
    )
    Radius:NiagaraFloat=&UniformRangedFloat(Minimum=6, Maximum=14)
    T:NiagaraFloat=$Particles.Partition.NormalizedID
   )
  ]
  particleUpdateScript [
   &ParticleState()
   &FN_BoneFlow(
    Angle:NiagaraFloat=&Add_Float(A=$Particles.Initial.FN_Initialize_BoneFlow_Attributes_v2.Angle, B=HlslExpression() {hlslExpression=Emitter.Age * .2})
    Height:NiagaraFloat=$Particles.Partition.NormalizedID
    Radius:NiagaraFloat=$Particles.Initial.FN_Initialize_BoneFlow_Attributes_v2.Radius
    SkeletalMesh:NiagaraDataInterfaceSkeletalMesh=$User.SkeletalMesh
    Use Sockets Instead=true
    Write Position Attribute=true
   )
   &Color(
    Color=&ColorFromCurve(ColorCurve=DataInterface() {propertyValues {
       AlphaCurve {keys [<at 0 0 RCIM_Cubic>,<at 0.1239 1>,<at 0.346 1>,<at 1 0 RCIM_Cubic>]}
       BlueCurve {keys [<at -0.006 1.2>,<at 0.261 0.2>]}
       GreenCurve {keys [<at -0.006 4.6059>,<at 0.261 0.7676>]}
       RedCurve {keys [<at -0.006 6>,<at 0.261 1>]}
      }})
    Scale Alpha=&FloatFromCurve(FloatCurve=DataInterface() {propertyValues {Curve {keys [<at 0.1282 1.0115>,<at 0.2455 0.0123>,<at 0.3389 0.9565>,<at 0.4758 0.9657>,<at 0.591 0.0032>,<at 0.6388 1.0115>,<at 0.8061 1>,<at 0.95 1>,<at 1 0>]}}})
    Scale Color=&VectorFromFloat(Value=&FloatFromCurve(
      FloatCurve=DataInterface() {propertyValues {Curve {keys [<at 0 3.5>,<at 0.2994 0>,<at 0.573 2>,<at 0.8425 0.0123>]}}}
      Scale Curve=&FN_FrontEndFloatModifier(Gameplay=1, Front End=0.5)
     ))
   )
   &DynamicMaterialParameters(
    Index 0 Param 2=&FloatFromCurve(FloatCurve=DataInterface() {propertyValues {Curve {keys [<at 0 0 RCIM_Cubic>, {arriveTangent=-6.7156, interpMode=RCIM_Cubic, leaveTangent=-6.7156, tangentMode=RCTM_User, time=1, value=-2}]}}})
    Module.Param0WriteEnabled=true
    Module.Param1WriteEnabled=true
    Module.Param2WriteEnabled=true
    Module.Param3WriteEnabled=true
   )
  ]
  rendererClasses [&NiagaraRibbonRendererProperties]
  renderers [{rendererClass=&NiagaraRibbonRendererProperties}]
 }
}
summary {
 systemName=NS_PIckaxe_LunchBox_Idle
 emitters [
  {emitterName=Flash_Flare, rendererClasses [&NiagaraLightRendererProperties]}
  {emitterName=LightningOuterArcs, rendererClasses [&NiagaraRibbonRendererProperties]}
  {emitterName=Electric_Strokes_Handle, rendererClasses [&NiagaraSpriteRendererProperties]}
  {emitterName=Electric_Strokes_Head, rendererClasses [&NiagaraSpriteRendererProperties]}
 ]
 userVariables [
  User.bIsFrontEnd {type=NiagaraBool}
  User.bIsFrontEndPreview {type=NiagaraBool}
  User.SkeletalMesh {
   type=NiagaraDataInterfaceSkeletalMesh
   defaultValue {
    struct=NiagaraExt_VariableValue_DataInterface
    value {dataInterface=&self:NiagaraDataInterfaceSkeletalMesh_0}
   }
  }
 ]
}
