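# Material for the one-handed Lunch Box pickaxe cosmetic: a Default Lit surface usable on skeletal meshes and mesh deformers.
# It uses no textures or parameters: one constant warm yellow drives both base color and emissive, with full specular and zero roughness for a glossy, glowing look.
# Named reroutes feed these values into a material attributes node, which passes through the shared MF_TechArtOverrides_accessories function.
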Material /BRCosmetics/Weapons/FORT_Melee/Pickaxe_LunchBox1H/Materials/M_Pickaxe_LunchBox_1h
graph {
 properties {
  bUseMaterialAttributes=true
  bUsedWithMeshDeformer=true
  bUsedWithSkeletalMesh=true
  pixelDepthOffsetMode=PDOM_Legacy
  usedShadingModels=Default Lit
 }
 nodes [
  n1 = Constant3Vector() {constant=<rgba 1 0.8716 0.2611 0>}
  n2 = SetMaterialAttributes!(
   MaterialAttributes
   Base Color=NamedRerouteDeclaration(None=@n1) {name=baseColor, nodeColor=<rgba 0.2422 1 0 1>}
   Specular=NamedRerouteDeclaration(None=Constant() {r=1}) {name=specular, nodeColor=<rgba 1 0 0.2344 1>}
   Roughness=NamedRerouteDeclaration(None=Constant()) {name=roughness, nodeColor=<rgba 0 1 0.0625 1>}
   Emissive Color=NamedRerouteDeclaration(None=@n1) {name=emissive, nodeColor=<rgba 0 0.5312 1 1>}
  ) {attributeSetTypes [
    #1
    #2
    #3
    #4
   ]}
  MaterialFunctionCall(In=@n2) {materialFunction=&MF_TechArtOverrides_accessories}
 ]
}
